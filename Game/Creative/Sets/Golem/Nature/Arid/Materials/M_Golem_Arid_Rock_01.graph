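# Default Lit surface material for the Golem set's Arid Rock 01 nature asset in Creative.
# Its graph holds three fixed texture samples: the base rock texture, a roughness map and a normal map.
# It has no parameters or math nodes, and the graph as shown lists no connections from the samples to material outputs.
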Material /Game/Creative/Sets/Golem/Nature/Arid/Materials/M_Golem_Arid_Rock_01
graph {
 properties {pixelDepthOffsetMode=PDOM_Legacy, usedShadingModels=Default Lit}
 nodes [
  TextureSample() {texture=/Game/Creative/Sets/Golem/Nature/Arid/Textures/T_Golem_Arid_Rock_01}
  TextureSample() {texture=/Game/Creative/Sets/Golem/Nature/Arid/Textures/T_Golem_Arid_Rock_01_Roughness}
  TextureSample() {samplerType=SAMPLERTYPE_Normal, texture=/Game/Creative/Sets/Golem/Nature/Arid/Textures/T_Golem_Arid_Rock_01_Normal}
 ]
}
